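# Unlit, two-sided translucent Niagara material for KnightCat emote FX, producing an eroding, particle-tinted shape.
# Emissive blends grey to white by a contrasted Fresnel, tinted by particle color.
# Opacity math offsets UV-driven gradient noise by a Dissolve dynamic parameter, sharpens it, then multiplies by particle alpha, view-angle fade and depth fade (bypassed on mobile).
# A panning normal-map distortion branch exists but does not feed the noise.
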Material /BRCosmetics/Animation/Game/MainPlayer/Emotes/KnightCat/FX/M_KC_Erosion
graph {
 properties {
  bUsedWithNiagaraMeshParticles=true
  bUsedWithNiagaraRibbons=true
  bUsedWithNiagaraSprites=true
  blendMode=BLEND_Translucent
  pixelDepthOffsetMode=PDOM_Legacy
  shadingModel=MSM_Unlit
  twoSided=true
  usedShadingModels=Unlit
 }
 nodes [
  n1 = MaterialFunctionCall(Float2=TextureCoordinate()) {materialFunction=&BreakOutFloat2Components}
  n2 = MaterialFunctionCall(X=@n1.R, Y=Add(A=Multiply(A=@n1.G) {constB=1.5}) {constB=0.25}) {materialFunction=&MakeFloat3}
  n3 = Multiply(A=TextureSample(UVs=Panner() {speedX=-0.1, speedY=-0.2}) {samplerType=SAMPLERTYPE_Normal, texture=&T_Weap_Distort_N_02}.RGB) {constB=0.35}
  Add(A=@n2.Result, B=ComponentMask(None=Saturate(None=@n3)) {b=true})
  Constant() {r=0.5}
  n4 = Constant3Vector() {constant=<rgba 0.5 0.5 0.5 0>}
  n5 = Constant3Vector() {constant=<rgba 1 1 1 0>}
  n6 = Fresnel(ExponentIn=Constant() {r=1})
  n7 = MaterialFunctionCall(In=@n6, Contrast=Constant() {r=0.5}) {materialFunction=&CheapContrast}
  n8 = Add(A=@n6, B=@n7.Result)
  n9 = Multiply(A=Saturate(None=Multiply(A=@n8, B=@n7.Result))) {constB=0.75}
  n10 = ParticleColor()
  MaterialFunctionCall(Emissive=Multiply(A=LinearInterpolate(A=@n4, B=@n5, Alpha=@n9), B=@n10.RGB)) {materialFunction=&MF_FX_EmissiveAdjustments}
  n11 = DepthFade(FadeDistance=ScalarParameter() {defaultValue=20, parameterName=Depth Fade Distance})
  n12 = ShadingPathSwitch(Default=@n11, Deferred=@n11, Mobile=Constant() {r=1}) {desc=Migrated from FeatureLevelSwitch}
  n13 = MaterialFunctionCall("FadeSensitivity (optional)"=Constant()) {materialFunction=&FadeBasedOnViewAngle}
  n14 = Noise!(World Position=@n2.Result, FilterWidth) {levelScale=3, levels=1, noiseFunction=NOISEFUNCTION_GradientTex3D, scale=15}
  n15 = DynamicParameter() {paramNames [Dissolve,Param2,Param3,Param4]}
  Multiply(
   A=Multiply(A=@n10.A, B=Multiply(A=@n12, B=@n13.Result))
   B=MaterialFunctionCall(In=Add(A=@n14, B=@n15.Dissolve), Contrast=Constant() {r=5}) {materialFunction=&CheapContrast}.Result
  )
 ]
}
