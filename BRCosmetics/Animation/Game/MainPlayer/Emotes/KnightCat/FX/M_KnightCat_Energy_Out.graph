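# Translucent, unlit, two-sided Niagara sprite material for energy effects in the KnightCat emote.
# Two noise textures pan in radial coordinates at Dynamic Parameter speeds, using explicit derivatives from pre-conversion UVs for mip selection.
# Their sum tints CrackColor over a UV1 sphere-masked BaseColor-to-CrackColor gradient, with desaturation, particle color and emissive adjustment.
# Alpha combines noise, a UV1 triangle mask, particle alpha and view-angle fade.
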Material /BRCosmetics/Animation/Game/MainPlayer/Emotes/KnightCat/FX/M_KnightCat_Energy_Out
graph {
 properties {
  bUsedWithNiagaraSprites=true
  blendMode=BLEND_Translucent
  pixelDepthOffsetMode=PDOM_Legacy
  shadingModel=MSM_Unlit
  twoSided=true
  usedShadingModels=Unlit
 }
 nodes [
  n1 = MaterialFunctionCall() {materialFunction=&VectorToRadialValue}
  n2 = DynamicParameter() {defaultValue=<rgba -1 0 -0.5 0>, paramNames [X01,Y01,X02,Y02]}
  n3 = ConstantBiasScale(None=TextureCoordinate())
  n4 = DDX(Value=@n3)
  n5 = DDY(Value=@n3)
  n6 = TextureSample!(
   UVs=Panner(Coordinate=@n1.Radial Coordinates, Speed=AppendVector(A=@n2.X01, B=@n2.X01))
   Tex
   "DDX(UVs)"=@n4
   "DDY(UVs)"=@n5
   Apply View MipBias
  ) {mipValueMode=TMVM_Derivative, samplerType=SAMPLERTYPE_Grayscale, texture=&LowResBlurredNoise}
  n7 = TextureSample(UVs=Panner(Coordinate=@n1.Vector Converted to Angle, Speed=AppendVector(A=@n2.X02, B=@n2.X02)), "DDX(UVs)"=@n4, "DDY(UVs)"=@n5) {
   mipValueMode=TMVM_Derivative
   samplerType=SAMPLERTYPE_LinearGrayscale
   texture=/Game/Global/Textures/Default/Noise/EqualizerComponents_v2_Noise
  }
  n8 = Add(A=@n6.RGB, B=@n7.RGB)
  n9 = Multiply(A=@n8) {constB=0.75}
  Add(A=@n9) {constB=0.7}
  n10 = ParticleColor()
  n11 = VectorParameter() {parameterName=CrackColor, defaultValue=<rgba 0.5 0.5 0.5 0>}
  n12 = SphereMask(A=TextureCoordinate() {coordinateIndex=1}, B=Constant2Vector() {g=0.5, r=1}) {attenuationRadius=0.5, hardnessPercent=0}
  n13 = Add(
   A=Multiply(A=@n8, B=Multiply(A=@n11.RGB) {constB=1.5})
   B=LinearInterpolate(A=VectorParameter() {parameterName=BaseColor, defaultValue=<rgba 0.1 0.1 0.1 0>}.RGB, B=@n11.RGB, Alpha=Power(Base=@n12))
  )
  n14 = ScalarParameter() {parameterName=Desaturation}
  MaterialFunctionCall(Emissive=Multiply(A=@n10.RGB, B=Desaturation(None=@n13, Fraction=@n14))) {materialFunction=&MF_FX_EmissiveAdjustments}
  n15 = Multiply(
   A=@n9
   B=TextureSample!(UVs=TextureCoordinate() {coordinateIndex=1}, Tex, Apply View MipBias) {texture=&/T_Triangle_Mask}.R
  )
  Multiply(
   A=Multiply(A=@n10.A, B=Multiply(A=@n15))
   B=MaterialFunctionCall("FadeSensitivity (optional)"=Constant() {r=1}) {materialFunction=&FadeBasedOnViewAngle}.Result
  )
 ]
}
