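# Niagara system for the KnightCat player emote: glowing blades on two skeletal mesh bones, timed energy bursts above the character.
# Blade mesh emitters and two small lights sample both bones of User.SkeletalMesh, with glow driven by AlphaFX_Glow and separate fade curves for front-end preview.
# Bits, flares, glows and an expanding ghost mesh burst at fixed offsets, delayed by Delay_End or fixed times.
# Most emitters tint from VariantColors, usually indexed by IndexColors.
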NiagaraSystem /BRCosmetics/Animation/Game/MainPlayer/Emotes/KnightCat/FX/NS_KnightCat_Emote
dependencies {
 usedDataInterfaces [
  &NiagaraDataInterfaceVectorCurve
  &NiagaraDataInterfaceCurve
  &NiagaraDataInterfaceColorCurve
  &NiagaraDataInterfaceMeshRendererInfo
  &NiagaraDataInterfaceVector2DCurve
 ]
 usedDynamicInputs [&Lerp_Float, &SetBoolByFloatComparison]
 usedModules [
  &SystemState
  &self:SystemSpawnScript.SystemScriptSource.SystemScriptGraph.NiagaraNodeAssignment_0.SetVariables_#1
  &Timeline
 ]
}
emitters {
 Bits {
  emitterUpdateScript [
   &EmitterState(Life Cycle Mode=::NewEnumerator1, Loop Behavior=::NewEnumerator1, Loop Delay=$User.Delay_End, Loop Duration=2.5, UseLoopDelay=true)
   &SpawnBurst_Instantaneous(Spawn Count=40, Use Spawn Probability=false)
  ]
  particleSpawnScript [
   &InitializeParticle(
    Color=&SelectLinearColorFromArray(Array Sampling Mode=::NewEnumerator1, Direct Array Index=$User.IndexColors, Color Selection Array=$User.VariantColors, Random Seed=?)
    Lifetime Max=1
    Lifetime Min=0.25
    Lifetime Mode=::NewEnumerator1
    Sprite Size Mode=::NewEnumerator1
    Uniform Sprite Size Max=5
    Uniform Sprite Size Min=1.5
    UsePositionOffset=false
    Write Material Random=true
   )
   &SphereLocation(
    MaskSpawnGroup=false
    Non Uniform Scale=<1 1 1.5>
    Offset=<0 20 200>
    Override Seed=false
    Sphere Radius=&RandomRangeFloat(Minimum=18, Maximum=30)
    Surface Only=false
    UseNonUniformScale=true
    UseOffset=true
    UseSphereOrientation=false
   )
   &AddVelocityInCone(
    Cone Angle=75
    Cone Axis=<0 0 1>
    Module.Use Velocity Falloff On Cone Axis=false
    Velocity Strength=&RandomRangeFloat(Minimum=60, Maximum=700)
   )
  ]
  particleUpdateScript [
   &ParticleState()
   &Drag(Drag=5, Use Linear Drag=true, Use Rotational Drag=false)
   &CurlNoiseForce(
    Noise Strength=&FloatFromCurve(
     FloatCurve=DataInterface() {propertyValues {Curve {keys [{interpMode=RCIM_Cubic, tangentMode=RCTM_Break, time=0, value=1}, {interpMode=RCIM_Cubic, tangentMode=RCTM_Break, time=1, value=0}]}}}
     Scale Curve=1000
    )
    Pan Noise=false
   )
   &SolveForcesAndVelocity(Clamp Velocity=false, Limit Acceleration=false)
   &ScaleColor(
    Scale Alpha=&FloatFromCurve(FloatCurve=DataInterface() {propertyValues {Curve {keys [
         {interpMode=RCIM_Cubic, tangentMode=RCTM_User, time=0, value=0}
         {interpMode=RCIM_Cubic, tangentMode=RCTM_User, time=0.15, value=1}
         {interpMode=RCIM_Cubic, tangentMode=RCTM_User, time=0.3, value=2.875}
         {interpMode=RCIM_Cubic, tangentMode=RCTM_User, time=0.4, value=1.25}
         {interpMode=RCIM_Cubic, tangentMode=RCTM_User, time=0.5, value=1.2}
         {interpMode=RCIM_Cubic, tangentMode=RCTM_User, time=0.6, value=2.125}
         {interpMode=RCIM_Cubic, tangentMode=RCTM_User, time=0.7, value=1.2}
         {interpMode=RCIM_Cubic, tangentMode=RCTM_User, time=0.8, value=1}
         {interpMode=RCIM_Cubic, tangentMode=RCTM_User, time=1, value=0}
        ]}}})
    Scale RGB=<10 10 10>
    ScaleA=true
    ScaleRGB=true
   ) {module=ScaleColor001}
   &ScaleSpriteSize(Uniform Curve Sprite Scale=DataInterface() {propertyValues {Curve {keys [
        {interpMode=RCIM_Cubic, tangentMode=RCTM_User, time=0, value=0.5}
        {interpMode=RCIM_Cubic, tangentMode=RCTM_User, time=0.1, value=1}
        {interpMode=RCIM_Cubic, tangentMode=RCTM_Break, time=1, value=0.2}
       ]}}}) {module=ScaleSpriteSize001}
   &CameraOffset(Camera Offset Amount=15)
  ]
  rendererClasses [&NiagaraSpriteRendererProperties]
  renderers [{rendererClass=&NiagaraSpriteRendererProperties}]
 }
 Blades_Ghost_Trail {
  emitterUpdateScript [
   &EmitterState(Life Cycle Mode=::NewEnumerator1, Loop Behavior=::NewEnumerator1, Loop Delay=$User.Delay_End, Loop Duration=2.5, UseLoopDelay=true)
   &SpawnBurst_Instantaneous(Use Spawn Probability=false, UseLoopCountLimit=false)
   &SpawnRate(SpawnRate=50, Use Spawn Probability=false) {enabled=false}
  ]
  particleSpawnScript [
   &InitializeParticle(
    Color=&SelectLinearColorFromArray(Array Sampling Mode=::NewEnumerator1, Direct Array Index=$User.IndexColors, Color Selection Array=$User.VariantColors, Random Seed=?)
    Lifetime=2.5
    Mesh Scale Mode=::NewEnumerator3
    UsePositionOffset=false
    Write Material Random=true
   )
   &InitialMeshOrientation(
    Mesh Axis To Orient=&RotateVectorByQuaternion(Vector To Rotate=<0.15 -0.15 -1>, Quaternion=$Particles.SkeletalMeshLocation.SampledRotation)
    Mesh Orientation Mode=::NewEnumerator2
    Orientation Vector=&RotateVectorByQuaternion(Vector To Rotate=<1 0 0>, Quaternion=$Particles.SkeletalMeshLocation.SampledRotation)
    Use Rotation Vector=false
   ) {enabled=false}
   /SetVariables(Particles.Position=&ConvertVectorToPosition(Input Position=<0 20 200>, Vector Is In Absolute Large World Space:NiagaraBool=false)) {isSetParameters=true}
   &FN_Mesh_Attributes(
    Color=&SelectLinearColorFromArray(Array Sampling Mode=::NewEnumerator1, Direct Array Index=$User.IndexColors, Color Selection Array=$User.VariantColors, Random Seed=?)
    Create Color Attribute=true
    "Create Normalized Age "=true
    Lifetime=1
    Write to Module Namespace=true
   )
   &InitialMeshOrientation(
    Mesh Orientation Mode=::NewEnumerator6
    Orientation Quaternion:Quat4f=&MakeQuatFromAxisAngle(Axis=<0 0 1>, AngleInDegrees=90)
    Use Rotation Vector=false
   ) {module=InitialMeshOrientation001}
  ]
  particleUpdateScript [
   &ParticleState()
   &ScaleColor(
    Linear Color Curve=DataInterface() {propertyValues {
      AlphaCurve {keys [<at 0 1>,<at 0.0202 1>,<at 0.2258 0.05>,<at 1 0>]}
      BlueCurve {keys [<at 0.0022 5>,<at 0.0374 1>,<at 0.5423 1>,<at 0.8523 0.5052>]}
      GreenCurve {keys [<at 0.0022 5>,<at 0.0374 1>,<at 0.5423 1>,<at 0.8523 0.5052>]}
      RedCurve {keys [<at 0.0022 5>,<at 0.0374 1>,<at 0.5423 1>,<at 0.8523 0.5052>]}
     }}
    Scale Mode=::NewEnumerator2
   )
   /SetVariables(Particles.Position=&ConvertVectorToPosition(
     Input Position=&Add_Vector(
      A=&ConvertPositionToVector(Input Position=$Particles.Position, Convert To Absolute Large World Vector=false)
      B=&VectorFromCurve(
       VectorCurve=DataInterface() {propertyValues {
         XCurve {keys [<at 0 0>,<at 1 0>]}
         YCurve {keys [<at 0 0>,<at 1 0>]}
         ZCurve {keys [
           <at 0 0 RCIM_Cubic>
           {arriveTangent=-13.7958, interpMode=RCIM_Cubic, leaveTangent=-13.7958, tangentMode=RCTM_User, time=0.0027, value=-1}
           <at 1 -5 RCIM_Cubic>
          ]}
        }}
       Scale Curve=<0.2 0.2 0.2>
      )
     )
     Vector Is In Absolute Large World Space:NiagaraBool=false
    )) {isSetParameters=true}
   /SetVariables(Particles.FN_Mesh_Attributes.Color:LinearColor=&ScaleLinearColorByCurve(
     InitialValue=$Particles.Initial.FN_Mesh_Attributes.Color
     ScaleCurve=DataInterface() {propertyValues {
       AlphaCurve {keys [<at 0 1>,<at 0.2079 1>,<at 0.4824 0.05>,<at 0.8219 0>]}
       BlueCurve {keys [<at 0.1587 20>,<at 0.2634 5>,<at 0.5423 1>,<at 0.8523 0.5052>]}
       GreenCurve {keys [<at 0.1587 20>,<at 0.2634 5>,<at 0.5423 1>,<at 0.8523 0.5052>]}
       RedCurve {keys [<at 0.1587 20>,<at 0.2634 5>,<at 0.5423 1>,<at 0.8523 0.5052>]}
      }}
     CurveIndex=$Particles.NormalizedAge
    )) {isSetParameters=true}
   &ScaleMeshSize(Scale Factor=&VectorFromFloat(Value=&FloatFromCurve(FloatCurve=DataInterface() {propertyValues {Curve {keys [
          {interpMode=RCIM_Cubic, tangentMode=RCTM_User, time=0, value=1}
          {interpMode=RCIM_Cubic, tangentMode=RCTM_User, time=0.0007, value=1.093}
          {arriveTangent=0.5748, interpMode=RCIM_Cubic, leaveTangent=0.5748, time=0.1827, value=1.207}
          {tangentMode=RCTM_User, time=0.9992, value=1.667}
         ]}}})))
   &DynamicMaterialParameters(
    Index 0 Param 1=&FloatFromCurve(FloatCurve=DataInterface() {propertyValues {Curve {keys [
         {arriveTangent=-22.5631, interpMode=RCIM_Cubic, leaveTangent=-22.5631, tangentMode=RCTM_User, time=0, value=2}
         {arriveTangent=-3.2349, interpMode=RCIM_Cubic, leaveTangent=-3.2349, tangentMode=RCTM_User, time=0.0552, value=1.0348}
         {arriveTangent=-9.1593, interpMode=RCIM_Cubic, leaveTangent=-9.1593, tangentMode=RCTM_User, time=0.8, value=-1}
        ]}}})
    Module.Param0WriteEnabled=true
    Module.Param1WriteEnabled=true
    Module.Param2WriteEnabled=true
    Module.Param3WriteEnabled=true
   )
  ]
  rendererClasses [&NiagaraMeshRendererProperties]
  renderers [
   {rendererClass=&NiagaraMeshRendererProperties, rendererIndex=0}
   {rendererClass=&NiagaraMeshRendererProperties, rendererIndex=1}
  ]
 }
 Blades_Ghost_Trail001 {
  emitterUpdateScript [
   &EmitterState(Life Cycle Mode=::NewEnumerator1, Loop Duration=0, UseLoopDelay=false)
   &SpawnBurst_Instantaneous(Use Spawn Probability=false, UseLoopCountLimit=false)
   &SpawnRate(SpawnRate=50, Use Spawn Probability=false) {enabled=false}
   &SpawnPerUnit(Module.Use Max Movement Threshold=true, Module.Use Movement Tolerance=false, Use Spawn Probability=false) {enabled=false}
  ]
  particleSpawnScript [
   &InitializeParticle(
    Color=&SelectLinearColorFromArray(Array Sampling Mode=::NewEnumerator1, Direct Array Index=$User.IndexColors, Color Selection Array=$User.VariantColors, Random Seed=?)
    Lifetime=0.5
    Mesh Scale Mode=::NewEnumerator3
    Mesh Uniform Scale=0.975
    UsePositionOffset=false
    Write Material Random=true
   )
   &InitialMeshOrientation(
    Mesh Axis To Orient=&RotateVectorByQuaternion(Vector To Rotate=<0.15 -0.15 -1>, Quaternion=$Particles.SkeletalMeshLocation.SampledRotation)
    Mesh Orientation Mode=::NewEnumerator2
    Orientation Vector=&RotateVectorByQuaternion(Vector To Rotate=<1 0 0>, Quaternion=$Particles.SkeletalMeshLocation.SampledRotation)
    Use Rotation Vector=false
   ) {enabled=false}
   &CameraOffset(Camera Offset Amount=-20) {enabled=false}
  ]
  particleUpdateScript [
   &SkeletalMeshLocation(Bone Sampling Mode=::NewEnumerator4, MaskSpawnGroup=false, Sampled Position Offset=<35 -3 -6>, Skeletal Mesh=$User.SkeletalMesh)
   /SetVariables(Particles.BladeOrientation:Quat4f=&MultiplyQuaternion(
     Quaternion A=&MakeQuatFromAxisAngle(
      Axis=&RotateVectorByQuaternion(Vector To Rotate=<-0.025 0.86 -0.1>, Quaternion=$Particles.SkeletalMeshLocation.SampledRotation)
      AngleInDegrees=-90
     )
     Quaternion B=$Particles.SkeletalMeshLocation.SampledRotation
    )) {isSetParameters=true, module=SetVariables_#2}
   &ParticleState()
   &ScaleColor(
    Linear Color Curve=DataInterface() {propertyValues {
      AlphaCurve {keys [<at 0 1>,<at 0.0654 1>,<at 0.1308 0.05>,<at 0.9264 0>]}
      BlueCurve {keys [<at 0.0022 5>,<at 0.1859 1>,<at 0.5423 1>,<at 0.8523 0.5052>]}
      GreenCurve {keys [<at 0.0022 5>,<at 0.1859 1>,<at 0.5423 1>,<at 0.8523 0.5052>]}
      RedCurve {keys [<at 0.0022 5>,<at 0.1859 1>,<at 0.5423 1>,<at 0.8523 0.5052>]}
     }}
    Scale Mode=::NewEnumerator2
   )
   &ScaleColor(Scale Alpha=&Multiply_Float(A=0.5, B=$User.AlphaFX_Glow), Scale RGB=<0.5 0.5 0.5>, ScaleA=true, ScaleRGB=true) {module=ScaleColor001}
   &ScaleMeshSize(Scale Factor=&MakeVector(
     X=&Multiply_Float(A=1, B=$User.AlphaFX_Glow)
     Y=&Multiply_Float(A=0.75, B=$User.AlphaFX_Glow)
     Z=1
    ))
  ]
  rendererClasses [&NiagaraMeshRendererProperties]
  renderers [{rendererClass=&NiagaraMeshRendererProperties, rendererIndex=0}]
 }
 Blades_Ghost_Trail002 {
  emitterUpdateScript [
   &EmitterState(Life Cycle Mode=::NewEnumerator1, Loop Duration=0.05, UseLoopDelay=false)
   &SpawnBurst_Instantaneous(Use Spawn Probability=false, UseLoopCountLimit=false)
  ]
  particleSpawnScript [
   &InitializeParticle(
    Color=&SelectLinearColorFromArray(Array Sampling Mode=::NewEnumerator1, Direct Array Index=$User.IndexColors, Color Selection Array=$User.VariantColors, Random Seed=?)
    Mesh Scale Mode=::NewEnumerator3
    Mesh Uniform Scale=0.975
    UsePositionOffset=false
    Write Material Random=true
   )
   &SkeletalMeshLocation(
    Bone Sampling Mode=::NewEnumerator4
    Interpolate Mesh Sampling=false
    MaskSpawnGroup=false
    Sampled Position Offset=<28.5 -5 -6>
    Skeletal Mesh=$User.SkeletalMesh
   )
   /SetVariables(Particles.BladeOrientation:Quat4f=&MultiplyQuaternion(
     Quaternion A=&MakeQuatFromAxisAngle(
      Axis=&RotateVectorByQuaternion(Vector To Rotate=<-0.025 0.86 -0.1>, Quaternion=$Particles.SkeletalMeshLocation.SampledRotation)
      AngleInDegrees=-90
     )
     Quaternion B=$Particles.SkeletalMeshLocation.SampledRotation
    )) {isSetParameters=true, module=SetVariables_#2}
   &FN_Mesh_Attributes(
    Color=$User.Variant_Color
    Create Color Attribute=true
    "Create Normalized Age "=true
    Create Orientation Attribute=true
    Lifetime=0.5
    Orientation:Quat4f=$Particles.MeshOrientation
    Write to Module Namespace=true
   )
  ]
  particleUpdateScript [
   &ParticleState()
   &ScaleColor(
    Scale Alpha=&MakeCustomFloatFromBool(
     Bool=$User.bIsFrontEndPreview
     True Float=&FloatFromCurve(
      FloatCurve=DataInterface() {propertyValues {Curve {keys [<at 0 1>,<at 0.0654 1>,<at 0.1792 0.075>,<at 0.9264 0>]}}}
      Scale Curve=$System.Timeline.Time
     )
     False Float=&FloatFromCurve(
      FloatCurve=DataInterface() {propertyValues {Curve {keys [<at 0 1>,<at 0.0654 1>,<at 0.1308 0.05>,<at 0.9264 0>]}}}
      Scale Curve=$System.Timeline.Time
     )
    )
    Scale RGB=&MakeCustomVectorFromBool(
     Bool=$User.bIsFrontEndPreview
     True Vector=&VectorFromCurve(VectorCurve=DataInterface() {propertyValues {
        XCurve {keys [<at 0.0022 4.5>,<at 0.3 4>,<at 0.55 1.5>,<at 0.8523 0.5052>]}
        YCurve {keys [<at 0.0022 4.5>,<at 0.3 4>,<at 0.55 1.5>,<at 0.8523 0.5052>]}
        ZCurve {keys [<at 0.0022 4.5>,<at 0.3 4>,<at 0.55 1.5>,<at 0.8523 0.5052>]}
       }})
     False Vector=&VectorFromCurve(VectorCurve=DataInterface() {propertyValues {
        XCurve {keys [<at 0.0022 5>,<at 0.1859 1>,<at 0.5423 1>,<at 0.8523 0.5052>]}
        YCurve {keys [<at 0.0022 5>,<at 0.1859 1>,<at 0.5423 1>,<at 0.8523 0.5052>]}
        ZCurve {keys [<at 0.0022 5>,<at 0.1859 1>,<at 0.5423 1>,<at 0.8523 0.5052>]}
       }})
    )
    ScaleA=true
    ScaleRGB=true
   )
   /SetVariables(Particles.Color_Glow:LinearColor=&MultiplyLinearColor(
     Linear Color=$Particles.Color
     Scale Factor=&MakeVector4_FromLinearColor(LinearColor=&MakeLinearColorFromVectorAndFloat("Vector (RGB)"=<0.25 0.25 0.25>, "Float (Alpha)"=$System.Timeline.Time))
    )) {isSetParameters=true}
   &ScaleMeshSize(Scale Factor=&VectorFromFloat(Value=1))
  ]
  rendererClasses [&NiagaraMeshRendererProperties]
  renderers [
   {rendererClass=&NiagaraMeshRendererProperties, rendererIndex=0}
   {rendererClass=&NiagaraMeshRendererProperties, rendererIndex=1}
  ]
 }
 Blades_Ghost_Trail004 {
  emitterUpdateScript [
   &EmitterState(Life Cycle Mode=::NewEnumerator1, Loop Duration=0, UseLoopDelay=false)
   &SpawnBurst_Instantaneous(Use Spawn Probability=false, UseLoopCountLimit=false)
   &SpawnRate(SpawnRate=50, Use Spawn Probability=false) {enabled=false}
   &SpawnPerUnit(Module.Use Max Movement Threshold=true, Module.Use Movement Tolerance=false, Use Spawn Probability=false) {enabled=false}
  ]
  particleSpawnScript [
   &InitializeParticle(
    Color=&SelectLinearColorFromArray(Array Sampling Mode=::NewEnumerator1, Direct Array Index=0, Color Selection Array=$User.VariantColors, Random Seed=?)
    Lifetime=0.5
    Mesh Scale Mode=::NewEnumerator3
    Mesh Uniform Scale=0.975
    UsePositionOffset=false
    Write Material Random=true
   )
   &InitialMeshOrientation(
    Mesh Axis To Orient=&RotateVectorByQuaternion(Vector To Rotate=<0.15 -0.15 -1>, Quaternion=$Particles.SkeletalMeshLocation.SampledRotation)
    Mesh Orientation Mode=::NewEnumerator2
    Orientation Vector=&RotateVectorByQuaternion(Vector To Rotate=<1 0 0>, Quaternion=$Particles.SkeletalMeshLocation.SampledRotation)
    Use Rotation Vector=false
   ) {enabled=false}
   &CameraOffset(Camera Offset Amount=-20) {enabled=false}
  ]
  particleUpdateScript [
   &SkeletalMeshLocation(
    Bone / Socket Index=1
    Bone Sampling Mode=::NewEnumerator4
    MaskSpawnGroup=false
    Sampled Position Offset=<35 -3 -6>
    Skeletal Mesh=$User.SkeletalMesh
   )
   /SetVariables(Particles.BladeOrientation:Quat4f=&MultiplyQuaternion(
     Quaternion A=&MakeQuatFromAxisAngle(
      Axis=&RotateVectorByQuaternion(Vector To Rotate=<-0.025 0.86 -0.1>, Quaternion=$Particles.SkeletalMeshLocation.SampledRotation)
      AngleInDegrees=-90
     )
     Quaternion B=$Particles.SkeletalMeshLocation.SampledRotation
    )) {isSetParameters=true, module=SetVariables_#2}
   &ParticleState()
   &ScaleColor(
    Linear Color Curve=DataInterface() {propertyValues {
      AlphaCurve {keys [<at 0 1>,<at 0.0654 1>,<at 0.1308 0.05>,<at 0.9264 0>]}
      BlueCurve {keys [<at 0.0022 5>,<at 0.1859 1>,<at 0.5423 1>,<at 0.8523 0.5052>]}
      GreenCurve {keys [<at 0.0022 5>,<at 0.1859 1>,<at 0.5423 1>,<at 0.8523 0.5052>]}
      RedCurve {keys [<at 0.0022 5>,<at 0.1859 1>,<at 0.5423 1>,<at 0.8523 0.5052>]}
     }}
    Scale Mode=::NewEnumerator2
   )
   &ScaleColor(Scale Alpha=&Multiply_Float(A=0.5, B=$User.AlphaFX_Glow), Scale RGB=<0.5 0.5 0.5>, ScaleA=true, ScaleRGB=true) {module=ScaleColor001}
   &ScaleMeshSize(Scale Factor=&MakeVector(
     X=&Multiply_Float(A=1, B=$User.AlphaFX_Glow)
     Y=&Multiply_Float(A=0.75, B=$User.AlphaFX_Glow)
     Z=1
    ))
  ]
  rendererClasses [&NiagaraMeshRendererProperties]
  renderers [{rendererClass=&NiagaraMeshRendererProperties, rendererIndex=0}]
 }
 Blades_Ghost_Trail005 {
  emitterUpdateScript [
   &EmitterState(Life Cycle Mode=::NewEnumerator1, Loop Duration=0.05, UseLoopDelay=false)
   &SpawnBurst_Instantaneous(Use Spawn Probability=false, UseLoopCountLimit=false)
  ]
  particleSpawnScript [
   &InitializeParticle(
    Color=&SelectLinearColorFromArray(Array Sampling Mode=::NewEnumerator1, Direct Array Index=$User.IndexColors, Color Selection Array=$User.VariantColors, Random Seed=?)
    Mesh Scale Mode=::NewEnumerator3
    Mesh Uniform Scale=0.975
    UsePositionOffset=false
    Write Material Random=true
   )
   &SkeletalMeshLocation(
    Bone / Socket Index=1
    Bone Sampling Mode=::NewEnumerator4
    Interpolate Mesh Sampling=false
    MaskSpawnGroup=false
    Sampled Position Offset=<28.5 -5 -6>
    Skeletal Mesh=$User.SkeletalMesh
   )
   /SetVariables(Particles.BladeOrientation:Quat4f=&MultiplyQuaternion(
     Quaternion A=&MakeQuatFromAxisAngle(
      Axis=&RotateVectorByQuaternion(Vector To Rotate=<-0.025 0.86 -0.1>, Quaternion=$Particles.SkeletalMeshLocation.SampledRotation)
      AngleInDegrees=-90
     )
     Quaternion B=$Particles.SkeletalMeshLocation.SampledRotation
    )) {isSetParameters=true}
   &FN_Mesh_Attributes(
    Color=$User.Variant_Color
    Create Color Attribute=true
    "Create Normalized Age "=true
    Create Orientation Attribute=true
    Lifetime=0.5
    Orientation:Quat4f=$Particles.MeshOrientation
    Write to Module Namespace=true
   )
  ]
  particleUpdateScript [
   &ParticleState()
   &ScaleColor(
    Scale Alpha=&MakeCustomFloatFromBool(
     Bool=$User.bIsFrontEndPreview
     True Float=&FloatFromCurve(
      FloatCurve=DataInterface() {propertyValues {Curve {keys [<at 0 1>,<at 0.0654 1>,<at 0.1792 0.075>,<at 0.9264 0>]}}}
      Scale Curve=$System.Timeline.Time
     )
     False Float=&FloatFromCurve(
      FloatCurve=DataInterface() {propertyValues {Curve {keys [<at 0 1>,<at 0.0654 1>,<at 0.1308 0.05>,<at 0.9264 0>]}}}
      Scale Curve=$System.Timeline.Time
     )
    )
    Scale RGB=&MakeCustomVectorFromBool(
     Bool=$User.bIsFrontEndPreview
     True Vector=&VectorFromCurve(VectorCurve=DataInterface() {propertyValues {
        XCurve {keys [<at 0.0022 4.5>,<at 0.2807 4.3714>,<at 0.5854 1.9408>,<at 0.8523 0.5052>]}
        YCurve {keys [<at 0.0022 4.5>,<at 0.2807 4.3714>,<at 0.5854 1.9408>,<at 0.8523 0.5052>]}
        ZCurve {keys [<at 0.0022 4.5>,<at 0.285 4.3714>,<at 0.5854 1.9408>,<at 0.8523 0.5052>]}
       }})
     False Vector=&VectorFromCurve(VectorCurve=DataInterface() {propertyValues {
        XCurve {keys [<at 0.0022 5>,<at 0.1859 1>,<at 0.5423 1>,<at 0.8523 0.5052>]}
        YCurve {keys [<at 0.0022 5>,<at 0.1859 1>,<at 0.5423 1>,<at 0.8523 0.5052>]}
        ZCurve {keys [<at 0.0022 5>,<at 0.1859 1>,<at 0.5423 1>,<at 0.8523 0.5052>]}
       }})
    )
    ScaleA=true
    ScaleRGB=true
   )
   /SetVariables(Particles.Color_Glow:LinearColor=&MultiplyLinearColor(
     Linear Color=$Particles.Color
     Scale Factor=&MakeVector4_FromLinearColor(LinearColor=&MakeLinearColorFromVectorAndFloat("Vector (RGB)"=<0.25 0.25 0.25>, "Float (Alpha)"=$System.Timeline.Time))
    )) {isSetParameters=true}
   &ScaleMeshSize(Scale Factor=&VectorFromFloat(Value=1))
  ]
  rendererClasses [&NiagaraMeshRendererProperties]
  renderers [
   {rendererClass=&NiagaraMeshRendererProperties, rendererIndex=0}
   {rendererClass=&NiagaraMeshRendererProperties, rendererIndex=1}
  ]
 }
 Flare {
  emitterUpdateScript [
   &EmitterState(Life Cycle Mode=::NewEnumerator1, Loop Behavior=::NewEnumerator1, Loop Delay=$User.Delay_End, Loop Duration=2.5, UseLoopDelay=true)
   &SpawnBurst_Instantaneous(Spawn Count=2, Use Spawn Probability=false)
  ]
  particleSpawnScript [
   &InitializeParticle(
    Color=&SelectLinearColorFromArray(Array Sampling Mode=::NewEnumerator1, Direct Array Index=$User.IndexColors, Color Selection Array=$User.VariantColors, Random Seed=?)
    Lifetime=0.25
    Sprite Size Mode=::NewEnumerator3
    Uniform Sprite Size=100
    UsePositionOffset=false
    Write Material Random=true
   )
   /SetVariables(Particles.Position=&ConvertVectorToPosition(Input Position=<0 20 225>, Vector Is In Absolute Large World Space:NiagaraBool=false)) {isSetParameters=true, module=SetVariables_#3}
   /SetVariables(Particles.FlareSize:Vector2f=<25 50>) {isSetParameters=true}
   /SetVariables(Particles.CameraOffset001=50) {isSetParameters=true}
   &CameraOffset(Camera Offset Amount=35)
  ]
  particleUpdateScript [
   &ParticleState()
   &ScaleColor(
    Scale Alpha=&FNTimeOfDayScalar(Night Scale=1, Morning Scale=1, Evening Scale=1)
    Scale RGB=&VectorFromFloat(Value=&FloatFromCurve(FloatCurve=DataInterface() {propertyValues {Curve {keys [<at 0 5>,<at 0.2 1>,<at 1 0>]}}}))
    ScaleA=true
    ScaleRGB=true
   )
   &ScaleSpriteSize(Scale Factor=&Vector2DFromFloat(Value=&FloatFromCurve(FloatCurve=DataInterface() {propertyValues {Curve {keys [<at 0 2>,<at 0.2 1>,<at 1 0>]}}})))
   /SetVariables(Particles.FlareColor=&MultiplyLinearColor(
     Linear Color=$Particles.Color
     Scale Factor=&MakeFloat4FromVectorAndFloat(
      "Vector (RGB)"=<5 5 5>
      "Float (Alpha)"=&FNTimeOfDayScalar(Night Scale=0.7, Morning Scale=0.5, Evening Scale=0.5)
     )
    )) {isSetParameters=true}
   /SetVariables(Particles.Color001:LinearColor=&MultiplyLinearColor(
     Linear Color=$Particles.FlareColor
     Scale Factor=&MakeFloat4FromVectorAndFloat(
      "Vector (RGB)"=&VectorFromFloat(Value=&FloatFromCurve(FloatCurve=DataInterface() {propertyValues {Curve {keys [<at 0 5>,<at 0.2 1>,<at 1 0>]}}}, Scale Curve=5))
      "Float (Alpha)"=&FNTimeOfDayScalar(Night Scale=0.7, Morning Scale=0.5, Evening Scale=0.7)
     )
    )) {isSetParameters=true}
   /SetVariables(Particles.SpriteSize001:Vector2f=&Multiply_Vector2(
     A=$Particles.FlareSize
     B=&Vector2DFromFloat(Value=&FloatFromCurve(FloatCurve=DataInterface() {propertyValues {Curve {keys [<at 0 2>,<at 0.2 1>,<at 1 0>]}}}))
    )) {isSetParameters=true}
   /SetVariables(
    Particles.RingColor:LinearColor=&MultiplyLinearColor(Linear Color=$Particles.Color, Scale Factor=<0.5 0.5 0.5 1>)
    Particles.RingSize:Vector2f=&MakeVector2D(
     X=&FloatFromCurve(
      FloatCurve=DataInterface() {propertyValues {Curve {keys [{tangentMode=RCTM_User, time=0, value=0}, {tangentMode=RCTM_User, time=0.1, value=1}, <at 1 3>]}}}
      Scale Curve=50
     )
     Y=&FloatFromCurve(
      FloatCurve=DataInterface() {propertyValues {Curve {keys [{tangentMode=RCTM_User, time=0, value=0}, {tangentMode=RCTM_User, time=0.1, value=1}, <at 1 3>]}}}
      Scale Curve=50
     )
    )
   ) {isSetParameters=true}
  ]
  rendererClasses [&NiagaraSpriteRendererProperties]
  renderers [
   {rendererClass=&NiagaraSpriteRendererProperties}
   {rendererClass=&NiagaraSpriteRendererProperties, rendererIndex=1}
   {rendererClass=&NiagaraSpriteRendererProperties, rendererIndex=2}
  ]
 }
 Flare001 {
  emitterUpdateScript [
   &EmitterState(Life Cycle Mode=::NewEnumerator1, Loop Behavior=::NewEnumerator1, Loop Delay=0.5, Loop Duration=2.5, UseLoopDelay=true)
   &SpawnBurst_Instantaneous(Spawn Count=2, Use Spawn Probability=false)
  ]
  particleSpawnScript [
   &InitializeParticle(
    Color=&SelectLinearColorFromArray(Array Sampling Mode=::NewEnumerator1, Direct Array Index=$User.IndexColors, Color Selection Array=$User.VariantColors, Random Seed=?)
    Lifetime=0.25
    Sprite Size=<50 150>
    Sprite Size Mode=::NewEnumerator4
    UsePositionOffset=false
    Write Material Random=true
   )
   /SetVariables(Particles.Position=&ConvertVectorToPosition(Input Position=<0 0 250>, Vector Is In Absolute Large World Space:NiagaraBool=false)) {isSetParameters=true, module=SetVariables_#3}
   &CameraOffset(Camera Offset Amount=35)
  ]
  particleUpdateScript [
   &ParticleState()
   &ScaleColor(
    Scale Alpha=&FNTimeOfDayScalar(Night Scale=0.5, Morning Scale=0.5, Day Scale=0.5, Evening Scale=0.5)
    Scale RGB=&VectorFromFloat(Value=&FloatFromCurve(
      FloatCurve=DataInterface() {propertyValues {Curve {keys [
          {arriveTangent=9.0725, interpMode=RCIM_Cubic, leaveTangent=9.0725, tangentMode=RCTM_User, time=0, value=0}
          {arriveTangent=-2.057, interpMode=RCIM_Cubic, leaveTangent=-2.057, tangentMode=RCTM_User, time=0.0331, value=0.9134}
          {arriveTangent=-0.0405, interpMode=RCIM_Cubic, leaveTangent=-0.0405, tangentMode=RCTM_User, time=1, value=0}
         ]}}}
      Scale Curve=0.5
     ))
    ScaleA=true
    ScaleRGB=true
   )
   &ScaleSpriteSize(Scale Factor=&Vector2DFromFloat(Value=&FloatFromCurve(FloatCurve=DataInterface() {propertyValues {Curve {keys [<at 0 2>,<at 0.2 1>,<at 1 0>]}}})))
  ]
  rendererClasses [&NiagaraSpriteRendererProperties]
  renderers [{rendererClass=&NiagaraSpriteRendererProperties}]
 }
 Glow {
  emitterUpdateScript [
   &EmitterState(Life Cycle Mode=::NewEnumerator1, Loop Behavior=::NewEnumerator1, Loop Delay=$User.Delay_End, Loop Duration=2.5, UseLoopDelay=true)
   &SpawnBurst_Instantaneous(Spawn Count=1, Use Spawn Probability=false)
  ]
  particleSpawnScript [
   &InitializeParticle(
    Color=&SelectLinearColorFromArray(Array Sampling Mode=::NewEnumerator1, Direct Array Index=$User.IndexColors, Color Selection Array=$User.VariantColors, Random Seed=?)
    Lifetime=0.2
    Sprite Size=<35 100>
    Sprite Size Mode=::NewEnumerator4
    UsePositionOffset=false
    Write Material Random=true
   )
   /SetVariables(Particles.Position=&ConvertVectorToPosition(Input Position=<0 20 225>, Vector Is In Absolute Large World Space:NiagaraBool=false)) {isSetParameters=true, module=SetVariables_#3}
  ]
  particleUpdateScript [
   &ParticleState()
   &ScaleSpriteSize(Scale Factor=&Vector2DFromFloat(Value=&FloatFromCurve(FloatCurve=DataInterface() {propertyValues {Curve {keys [
          {arriveTangent=0.4346, interpMode=RCIM_Cubic, leaveTangent=0.4346, tangentMode=RCTM_User, time=0, value=0}
          {arriveTangent=0.0473, interpMode=RCIM_Cubic, leaveTangent=0.0473, tangentMode=RCTM_User, time=0.2, value=1}
         ]}}})))
   &ScaleColor(
    Scale Alpha=&FloatFromCurve(
     FloatCurve=DataInterface() {propertyValues {Curve {keys [
         {arriveTangent=0.4005, interpMode=RCIM_Cubic, leaveTangent=0.4005, tangentMode=RCTM_User, time=0, value=0}
         {arriveTangent=0.0256, interpMode=RCIM_Cubic, leaveTangent=0.0256, tangentMode=RCTM_User, time=0.2, value=1}
         <at 1 0 RCIM_Cubic>
        ]}}}
     Scale Curve=&FNTimeOfDayScalar(Night Scale=0.8, Morning Scale=0.8, Evening Scale=0.8)
    )
    Scale RGB=<0.5 0.5 0.5>
    ScaleA=true
    ScaleRGB=true
   )
   /SetVariables(
    Particles.GlowColor2:LinearColor=&MakeLinearColorFromVectorAndFloat(
     "Vector (RGB)"=&MakeVectorFromLinearColorRGB(Color=$Particles.Color)
     "Float (Alpha)"=&Multiply_Float(
      A=0.25
      B=&FloatFromCurve(FloatCurve=DataInterface() {propertyValues {Curve {keys [
           {arriveTangent=0.4005, interpMode=RCIM_Cubic, leaveTangent=0.4005, tangentMode=RCTM_User, time=0, value=0}
           {arriveTangent=0.0256, interpMode=RCIM_Cubic, leaveTangent=0.0256, tangentMode=RCTM_User, time=0.2, value=1}
           <at 1 0 RCIM_Cubic>
          ]}}})
     )
    )
    Particles.GlowSize2:Vector2f=&MakeVector2D(
     X=&FloatFromCurve(
      FloatCurve=DataInterface() {propertyValues {Curve {keys [{tangentMode=RCTM_User, time=0, value=0}, {tangentMode=RCTM_User, time=0.1, value=1}, <at 1 3>]}}}
      Scale Curve=35
     )
     Y=&FloatFromCurve(
      FloatCurve=DataInterface() {propertyValues {Curve {keys [{tangentMode=RCTM_User, time=0, value=0}, {tangentMode=RCTM_User, time=0.1, value=1}, <at 1 3>]}}}
      Scale Curve=100
     )
    )
   ) {isSetParameters=true}
   /SetVariables(
    Particles.GlowColor03:LinearColor=&MultiplyLinearColor(
     Linear Color=$Particles.Color
     Scale Factor=&MakeFloat4FromVectorAndFloat(
      "Vector (RGB)"=<0.5 0.5 0.5>
      "Float (Alpha)"=&FNTimeOfDayScalar(Night Scale=0.3, Morning Scale=0.3, Day Scale=0.7, Evening Scale=0.5)
     )
    )
    Particles.GlowSize03:Vector2f=&MakeVector2D(
     X=&FloatFromCurve(
      FloatCurve=DataInterface() {propertyValues {Curve {keys [{tangentMode=RCTM_User, time=0, value=0}, {tangentMode=RCTM_User, time=0.1, value=1}, <at 1 3>]}}}
      Scale Curve=35
     )
     Y=&FloatFromCurve(
      FloatCurve=DataInterface() {propertyValues {Curve {keys [{tangentMode=RCTM_User, time=0, value=0}, {tangentMode=RCTM_User, time=0.1, value=1}, <at 1 3>]}}}
      Scale Curve=100
     )
    )
   ) {isSetParameters=true}
  ]
  rendererClasses [&NiagaraSpriteRendererProperties]
  renderers [
   {rendererClass=&NiagaraSpriteRendererProperties}
   {rendererClass=&NiagaraSpriteRendererProperties, rendererIndex=1}
   {rendererClass=&NiagaraSpriteRendererProperties, rendererIndex=2}
  ]
 }
 Glow001 {
  emitterUpdateScript [
   &EmitterState(Life Cycle Mode=::NewEnumerator1, Loop Behavior=::NewEnumerator1, Loop Delay=0.45, Loop Duration=2.5, UseLoopDelay=true)
   &SpawnBurst_Instantaneous(Spawn Count=1, Use Spawn Probability=false)
  ]
  particleSpawnScript [
   &InitializeParticle(
    Color=&SelectLinearColorFromArray(Array Sampling Mode=::NewEnumerator1, Direct Array Index=$User.IndexColors, Color Selection Array=$User.VariantColors, Random Seed=?)
    Lifetime=0.5
    Sprite Size=<35 100>
    Sprite Size Mode=::NewEnumerator4
    UsePositionOffset=false
    Write Material Random=true
   )
   /SetVariables(Particles.Position=&ConvertVectorToPosition(Input Position=<0 0 250>, Vector Is In Absolute Large World Space:NiagaraBool=false)) {isSetParameters=true, module=SetVariables_#3}
  ]
  particleUpdateScript [
   &ParticleState()
   &ScaleSpriteSize(Scale Factor=&Vector2DFromFloat(Value=&FloatFromCurve(FloatCurve=DataInterface() {propertyValues {Curve {keys [
          {arriveTangent=0.4346, interpMode=RCIM_Cubic, leaveTangent=0.4346, tangentMode=RCTM_User, time=0, value=0}
          {arriveTangent=0.0473, interpMode=RCIM_Cubic, leaveTangent=0.0473, tangentMode=RCTM_User, time=0.1, value=1.25}
          <at 0.9986 1.9316>
         ]}}})))
   &ScaleColor(
    Scale Alpha=&FloatFromCurve(
     FloatCurve=DataInterface() {propertyValues {Curve {keys [
         {arriveTangent=0.4005, interpMode=RCIM_Cubic, leaveTangent=0.4005, tangentMode=RCTM_User, time=0, value=0}
         {arriveTangent=7.3692, interpMode=RCIM_Cubic, leaveTangent=7.3692, time=0.025, value=10.6542}
         {arriveTangent=0.0256, interpMode=RCIM_Cubic, leaveTangent=0.0256, tangentMode=RCTM_User, time=0.0501, value=0.3689}
         <at 1 0 RCIM_Cubic>
        ]}}}
     Scale Curve=&FNTimeOfDayScalar(Night Scale=0.8, Morning Scale=0.8, Evening Scale=0.8)
    )
    ScaleA=true
    ScaleRGB=true
   )
  ]
  rendererClasses [&NiagaraSpriteRendererProperties]
  renderers [{rendererClass=&NiagaraSpriteRendererProperties}]
 }
 Glow002 {
  emitterUpdateScript [
   &EmitterState(Life Cycle Mode=::NewEnumerator1, Loop Behavior=::NewEnumerator1, Loop Delay=0.5, Loop Duration=2.5, UseLoopDelay=true)
   &SpawnBurst_Instantaneous(Spawn Count=1, Use Spawn Probability=false)
  ]
  particleSpawnScript [
   &InitializeParticle(
    Color=&SelectLinearColorFromArray(Array Sampling Mode=::NewEnumerator1, Direct Array Index=$User.IndexColors, Color Selection Array=$User.VariantColors, Random Seed=?)
    Lifetime=0.35
    Sprite Size=<35 100>
    Sprite Size Mode=::NewEnumerator4
    UsePositionOffset=false
    Write Material Random=true
   )
   /SetVariables(Particles.Position=&ConvertVectorToPosition(Input Position=<0 0 250>, Vector Is In Absolute Large World Space:NiagaraBool=false)) {isSetParameters=true, module=SetVariables_#3}
  ]
  particleUpdateScript [
   &ParticleState()
   &ScaleSpriteSize(Scale Factor=&Vector2DFromFloat(Value=&FloatFromCurve(
      FloatCurve=DataInterface() {propertyValues {Curve {keys [
          {arriveTangent=52.4933, interpMode=RCIM_Cubic, leaveTangent=52.4933, tangentMode=RCTM_User, time=0, value=0}
          {arriveTangent=2.4937, interpMode=RCIM_Cubic, leaveTangent=2.4937, tangentMode=RCTM_User, time=0.0169, value=0.4527}
          <at 1 2>
         ]}}}
      Scale Curve=2
     )))
   &ScaleColor(
    Scale Alpha=&FloatFromCurve(FloatCurve=DataInterface() {propertyValues {Curve {keys [
         {arriveTangent=2761.8682, interpMode=RCIM_Cubic, leaveTangent=2761.8682, tangentMode=RCTM_User, time=0, value=0}
         {arriveTangent=-337.6998, interpMode=RCIM_Cubic, leaveTangent=-337.6998, tangentMode=RCTM_User, time=0.025, value=10.6542}
         {arriveTangent=-2.9571, interpMode=RCIM_Cubic, leaveTangent=-2.9571, tangentMode=RCTM_User, time=0.0289, value=0.8746}
         <at 1 0 RCIM_Cubic>
        ]}}})
    ScaleA=true
    ScaleRGB=true
   )
   &DynamicMaterialParameters(
    Index 0 Param 1=-0.25
    Index 0 Param 2=-0.25
    Index 0 Param 3=0.5
    Index 0 Param 4=0.5
    Module.Param0WriteEnabled=true
    Module.Param1WriteEnabled=true
    Module.Param2WriteEnabled=true
    Module.Param3WriteEnabled=true
   )
  ]
  rendererClasses [&NiagaraSpriteRendererProperties]
  renderers [{rendererClass=&NiagaraSpriteRendererProperties}]
 }
 Light001 {
  emitterUpdateScript [
   &EmitterState(Life Cycle Mode=::NewEnumerator1, Loop Behavior=::NewEnumerator1, Loop Delay=0.5, Loop Duration=2.5, UseLoopDelay=true)
   &SpawnBurst_Instantaneous(Spawn Group=1, Use Spawn Probability=false, UseLoopCountLimit=false) {module=SpawnBurst_Instantaneous001}
  ]
  particleSpawnScript [&InitializeParticle(
    Color=&SelectLinearColorFromArray(Array Sampling Mode=::NewEnumerator1, Direct Array Index=$User.IndexColors, Color Selection Array=$User.VariantColors, Random Seed=?)
    Lifetime=4
    UsePositionOffset=false
    Write Material Random=true
   )]
  particleUpdateScript [
   &ParticleState()
   &SkeletalMeshLocation(Bone Sampling Mode=::NewEnumerator4, MaskSpawnGroup=false, Skeletal Mesh=$User.SkeletalMesh)
   &FN_Light_Attributes(
    Color=<rgba 2 0 5 1>
    Create Color Attribute=true
    Create Exponent Attribute=true
    Create Radius Attribute=true
    Exponent=0.5
    Radius=7.5
    Scale Color=&FloatFromCurve(FloatCurve=DataInterface() {propertyValues {Curve {keys [
         {interpMode=RCIM_Cubic, tangentMode=RCTM_User, time=0, value=0.5}
         {arriveTangent=1.3757, interpMode=RCIM_Cubic, leaveTangent=1.3757, tangentMode=RCTM_User, time=0.6, value=0.75}
         <at 1.0042 1>
        ]}}})
   )
   &CameraOffset(Camera Offset Amount=50)
  ]
  rendererClasses [&NiagaraLightRendererProperties]
  renderers [{rendererClass=&NiagaraLightRendererProperties, rendererIndex=0}]
 }
 Light002 {
  emitterUpdateScript [
   &EmitterState(Life Cycle Mode=::NewEnumerator1, Loop Behavior=::NewEnumerator1, Loop Delay=0.5, Loop Duration=2.5, UseLoopDelay=true)
   &SpawnBurst_Instantaneous(Spawn Group=1, Use Spawn Probability=false, UseLoopCountLimit=false) {module=SpawnBurst_Instantaneous001}
  ]
  particleSpawnScript [&InitializeParticle(
    Color=&SelectLinearColorFromArray(Array Sampling Mode=::NewEnumerator1, Direct Array Index=$User.IndexColors, Color Selection Array=$User.VariantColors, Random Seed=?)
    Lifetime=4
    UsePositionOffset=false
    Write Material Random=true
   )]
  particleUpdateScript [
   &ParticleState()
   &SkeletalMeshLocation(Bone / Socket Index=1, Bone Sampling Mode=::NewEnumerator4, MaskSpawnGroup=false, Skeletal Mesh=$User.SkeletalMesh)
   &FN_Light_Attributes(
    Color=<rgba 2 0 5 1>
    Create Color Attribute=true
    Create Exponent Attribute=true
    Create Radius Attribute=true
    Exponent=0.5
    Radius=7.5
    Scale Color=&FloatFromCurve(FloatCurve=DataInterface() {propertyValues {Curve {keys [
         {interpMode=RCIM_Cubic, tangentMode=RCTM_User, time=0, value=0.5}
         {arriveTangent=1.3757, interpMode=RCIM_Cubic, leaveTangent=1.3757, tangentMode=RCTM_User, time=0.6, value=0.75}
         <at 1.0042 1>
        ]}}})
   )
   &CameraOffset(Camera Offset Amount=50)
  ]
  rendererClasses [&NiagaraLightRendererProperties]
  renderers [{rendererClass=&NiagaraLightRendererProperties, rendererIndex=0}]
 }
}
summary {
 systemName=NS_KnightCat_Emote
 emitters [
  {emitterName=Blades_Ghost_Trail005, rendererClasses [&NiagaraMeshRendererProperties]}
  {emitterName=Blades_Ghost_Trail, rendererClasses [&NiagaraMeshRendererProperties]}
  {emitterName=Bits, rendererClasses [&NiagaraSpriteRendererProperties]}
  {emitterName=Glow, rendererClasses [&NiagaraSpriteRendererProperties]}
  {emitterName=Flare, rendererClasses [&NiagaraSpriteRendererProperties]}
  {emitterName=Blades_Ghost_Trail002, rendererClasses [&NiagaraMeshRendererProperties]}
  {emitterName=Blades_Ghost_Trail001, rendererClasses [&NiagaraMeshRendererProperties]}
  {emitterName=Blades_Ghost_Trail004, rendererClasses [&NiagaraMeshRendererProperties]}
  {emitterName=Glow001, rendererClasses [&NiagaraSpriteRendererProperties]}
  {emitterName=Light001, rendererClasses [&NiagaraLightRendererProperties]}
  {emitterName=Light002, rendererClasses [&NiagaraLightRendererProperties]}
  {emitterName=Glow002, rendererClasses [&NiagaraSpriteRendererProperties]}
  {emitterName=Flare001, rendererClasses [&NiagaraSpriteRendererProperties]}
 ]
 userVariables [
  User.AlphaFX_Blade {type=NiagaraFloat, defaultValue=0, description=""}
  User.AlphaFX_Glow {type=NiagaraFloat, defaultValue=0, description=""}
  User.bIsFrontEnd {type=NiagaraBool}
  User.bIsFrontEndPreview {type=NiagaraBool}
  User.Delay_End {type=NiagaraFloat, defaultValue=3.45, description=""}
  User.IndexColors {type=NiagaraInt32}
  User.Overheat {type=NiagaraFloat, defaultValue=0, description=""}
  User.SkeletalMesh {
   type=NiagaraDataInterfaceSkeletalMesh
   defaultValue {
    struct=NiagaraExt_VariableValue_DataInterface
    value {dataInterface=&self:NiagaraDataInterfaceSkeletalMesh_1}
   }
  }
  User.Variant_Color {type=LinearColor, description="", defaultValue {struct=LinearColor, value=<rgba 1 0 1 1>}}
  User.VariantColors {
   type=NiagaraDataInterfaceArrayColor
   description=""
   defaultValue {
    struct=NiagaraExt_VariableValue_DataInterface
    value {
     dataInterface=&self:NiagaraDataInterfaceArrayColor_0
     dataInterfaceClass=&NiagaraDataInterfaceArrayColor
    }
   }
  }
 ]
}
